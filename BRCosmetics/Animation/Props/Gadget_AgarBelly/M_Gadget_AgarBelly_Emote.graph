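# Two-sided lit material for the AgarBelly gadget emote prop, animating a wobbling skeletal mesh surface.
# World position offset sums three noise and mask layers pushed along normalized pre-skinned local position, with EmoteSpeed modulating two layers and WPOHideMesh blending against a collapse to the pivot.
# Color samples the diffuse texture using noise-driven Fresnel as UVs, adding rim brightening and faint glitched noise.
# Fresnel also blends metallic and specular ranges.
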Material /BRCosmetics/Animation/Props/Gadget_AgarBelly/M_Gadget_AgarBelly_Emote
graph {
 properties {
  bAutomaticallySetUsageInEditor=false
  bUsedWithMeshDeformer=true
  bUsedWithMorphTargets=true
  bUsedWithSkeletalMesh=true
  twoSided=true
  usedShadingModels=Default Lit
 }
 nodes [
  n1 = Fresnel(ExponentIn=Constant() {r=1.7106})
  LinearInterpolate(
   A=ScalarParameter() {defaultValue=1, parameterName=MetallicMin, sliderMax=1}
   B=ScalarParameter() {defaultValue=0.7, parameterName=MetallicMax, sliderMax=1}
   Alpha=@n1
  ) {constA=1}
  LinearInterpolate(
   A=ScalarParameter() {defaultValue=0.05, parameterName=SpecularMin, sliderMax=1}
   B=ScalarParameter() {defaultValue=1, parameterName=SpecularMax, sliderMax=1}
   Alpha=@n1
  )
  LinearInterpolate(
   A=Subtract(A=ObjectPositionWS(), B=WorldPosition().XYZ)
   B=NamedRerouteUsage()
   Alpha=ScalarParameter() {defaultValue=1, parameterName=WPOHideMesh}
  ) {constAlpha=1}
  LocalPosition() {includedOffsets=ExcludeOffsets, localOrigin=InstancePreSkinning}
  LocalPosition() {includedOffsets=ExcludeOffsets, localOrigin=InstancePreSkinning}
  n2 = Add(A=NamedRerouteUsage(), B=NamedRerouteUsage())
  n3 = NamedRerouteUsage()
  NamedRerouteDeclaration(None=Multiply(A=Add(A=@n2, B=@n3), B=ScalarParameter() {defaultValue=1, parameterName=WPO_Blend})) {name=WPO_Noise, nodeColor=<rgba 0 0.4375 1 1>}
  n4 = NamedRerouteDeclaration(None=Divide(
    A=LocalPosition() {includedOffsets=ExcludeOffsets, localOrigin=InstancePreSkinning}.XYZ
    B=ScalarParameter() {defaultValue=1, parameterName=LocalPositionScale}
   )) {name=LocalPositionScale, nodeColor=<rgba 0.0078 0 1 1>}
  n5 = ComponentMask(None=Multiply(A=@n4, B=ScalarParameter() {defaultValue=1, parameterName=Layer1_Scale})) {g=true, r=true}
  n6 = ScalarParameter() {defaultValue=1, parameterName=Layer1_UVTiling}
  n7 = Multiply(
   A=AppendVector(A=ScalarParameter() {defaultValue=0.5, parameterName=Layer1_SpeedX}, B=ScalarParameter() {defaultValue=0.5, parameterName=Layer1_SpeedY})
   B=ScalarParameter() {defaultValue=1}
  )
  n8 = Subtract(A=ScalarParameter() {defaultValue=1, parameterName=EmoteSpeed})
  n9 = LinearInterpolate(Alpha=@n8) {constA=1, constB=1.2}
  n10 = TextureSample(UVs=Panner(Coordinate=Multiply(A=@n5, B=@n6), Speed=Multiply(A=@n7, B=@n9)) {speedX=0.2, speedY=0.2}) {texture=/Game/Global/Textures/Default/Noise/TankMoteNoise}
  n11 = ComponentMask(None=ConstantBiasScale(None=@n10.R) {bias=-1, scale=2}) {r=true}
  n12 = Normalize(VectorInput=@n4)
  n13 = ScalarParameter() {defaultValue=3, parameterName=Layer1_Offset}
  n14 = LinearInterpolate(Alpha=@n8) {constA=1, constB=1.2}
  NamedRerouteDeclaration(None=Multiply(A=Multiply(A=@n11, B=@n12), B=Multiply(A=@n13, B=@n14))) {name=Layer1, nodeColor=<rgba 0 1 0.3672 1>}
  n15 = NamedRerouteUsage()
  n16 = ScalarParameter() {defaultValue=0.5, parameterName=Layer2_offsetY}
  n17 = Multiply(A=Time(), B=ScalarParameter() {defaultValue=2, parameterName=Layer2Speed})
  n18 = Add(
   A=Multiply(A=ComponentMask(None=@n15) {b=true, g=true}, B=ScalarParameter() {defaultValue=0.1, parameterName=Layer2_UVTiling})
   B=AppendVector(A=ScalarParameter() {parameterName=Layer2_OffsetX}, B=Add(A=@n16, B=@n17))
  )
  n19 = Multiply(
   A=TextureSample(UVs=@n18) {samplerType=SAMPLERTYPE_Masks, texture=&/T_Gadget_AgarBelly_Emote_WPOMask}.R
   B=ScalarParameter() {defaultValue=0.8, parameterName=Param_2}
  ) {constB=0.8}
  n20 = Multiply(
   A=TextureSample(UVs=@n18) {samplerType=SAMPLERTYPE_Masks, texture=&T_Gadget_AgarBelly_Emote_WPOMask}.G
   B=ScalarParameter() {defaultValue=2, parameterName=Param_1}
  ) {constB=2}
  n21 = Multiply(A=Normalize(VectorInput=@n15), B=ConstantBiasScale(None=Add(A=@n19, B=@n20)) {bias=-0.5, scale=1.1})
  n22 = LinearInterpolate(
   A=ScalarParameter() {defaultValue=5, parameterName=Layer2_Offset_Min}
   B=ScalarParameter() {defaultValue=4, parameterName=Layer2_Offset_Max}
   Alpha=Subtract(A=ScalarParameter() {defaultValue=1, parameterName=EmoteSpeed})
  ) {constB=4}
  NamedRerouteDeclaration(None=Multiply(A=Multiply(A=@n21, B=@n22), B=ScalarParameter() {defaultValue=1, parameterName=Param_3}) {constB=0}) {name=Layer2, nodeColor=<rgba 1 0.0469 0 1>}
  n23 = NamedRerouteUsage()
  n24 = ScalarParameter() {defaultValue=1, parameterName=Layer3_Scale}
  n25 = Multiply(A=ComponentMask(None=Multiply(A=@n23, B=@n24)) {g=true, r=true}, B=ScalarParameter() {defaultValue=0.01, parameterName=Layer3_UVTiling})
  n26 = Multiply(
   A=AppendVector(A=ScalarParameter() {defaultValue=0.1, parameterName=Layer3_SpeedX}, B=ScalarParameter() {defaultValue=0.1, parameterName=Layer3_SpeedY})
   B=ScalarParameter() {defaultValue=0.5, parameterName=Param_5}
  )
  n27 = OneMinus(None=TextureSample(UVs=Panner(Coordinate=@n25, Speed=@n26) {speedX=0.2, speedY=0.2}) {texture=/Game/Global/Textures/Default/Noise/T_SoftCells_Negative}.R)
  n28 = ComponentMask(None=ConstantBiasScale(None=Saturate(None=@n27)) {bias=-0.2}) {r=true}
  n29 = Normalize(VectorInput=@n23)
  NamedRerouteDeclaration(None=Multiply(A=Multiply(A=@n28, B=@n29), B=ScalarParameter() {defaultValue=5, parameterName=Layer3_Offset})) {name=Layer3, nodeColor=<rgba 0 1 0.3672 1>}
  Power(Base=@n10.R)
  n30 = Multiply(A=Panner(Coordinate=TextureCoordinate(), Speed=Constant() {r=5}) {speedX=0.1, speedY=0.1}, B=Constant() {r=0.1}) {constB=0.1}
  n31 = Fresnel(ExponentIn=LinearInterpolate(
    A=Constant() {r=3}
    B=Constant() {r=0.5}
    Alpha=TextureSample(UVs=@n30) {samplerType=SAMPLERTYPE_Grayscale, texture=/Game/Global/Textures/Default/Noise/LowResBlurredNoiseNormalized}.R
   ))
  n32 = ScalarParameter() {defaultValue=2, parameterName=Tiling}
  n33 = Panner(Coordinate=TextureCoordinate()) {speedX=0.03, speedY=0.03}
  n34 = Multiply(
   A=TextureSample(UVs=Multiply(A=@n31, B=@n32)) {texture=&/T_Gadget_AgarBelly_Emote_D}.RGB
   B=LinearInterpolate(Alpha=TextureSample(UVs=@n33) {samplerType=SAMPLERTYPE_Grayscale, texture=&LowResBlurredNoiseNormalized}.R) {constA=0.8}
  ) {constB=0.5}
  n35 = Constant() {r=2}
  n36 = LinearInterpolate(
   A=Multiply(A=@n34, B=Constant() {r=0.08}) {constB=0.5}
   B=Multiply(A=@n34, B=Constant() {r=3})
   Alpha=Power(Base=Fresnel(ExponentIn=@n35), Exp=Constant() {r=2})
  )
  Reroute(None=Multiply(A=@n36, B=Constant() {r=0.2}) {constB=0.2})
  n37 = TextureCoordinate()
  n38 = Constant() {r=10}
  n39 = TextureSample(UVs=MaterialFunctionCall!(UV=Multiply(A=@n37, B=@n38) {constB=4}) {materialFunction=/Game/Weapons/WeaponSkins/MaterialFunction/MF_GlitchFunction}.Glitched UV) {samplerType=SAMPLERTYPE_LinearColor, texture=/Game/Global/Textures/Default/Noise/T_Curie_NoisePack}
  n40 = Constant() {r=0.002}
  Reroute(None=Add(A=Multiply(A=@n39.RGB, B=@n40), B=@n36))
  ScalarParameter() {defaultValue=1, parameterName=Param_4}
  ScalarParameter() {defaultValue=0.25, parameterName=Roughness, sliderMax=1}
 ]
}
